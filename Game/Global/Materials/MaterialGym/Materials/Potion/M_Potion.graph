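# M_Potion is a default-lit surface material for potions in the MaterialGym folder.
# It uses material attributes and delegates all shading to the shared MF_Potion function from the Fortress_SharedMaterials package.
# The function takes an empty SetMaterialAttributes node as input, so this asset is a thin wrapper with no logic or overrides of its own.
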Material /Game/Global/Materials/MaterialGym/Materials/Potion/M_Potion
graph {
 properties {bUseMaterialAttributes=true, pixelDepthOffsetMode=PDOM_Legacy, usedShadingModels=Default Lit}
 nodes [MaterialFunctionCall!(In=SetMaterialAttributes!(MaterialAttributes)) {materialFunction=/Game/Packages/Fortress_SharedMaterials/Base_Material_Functions/MF_Potion}]
}
